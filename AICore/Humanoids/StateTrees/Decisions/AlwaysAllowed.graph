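# StateTree AlwaysAllowed, a humanoid AI decision asset that always succeeds.
# A single Root state runs a zero-duration Delay Task.
# When that state completes, its only transition ends the tree as Succeeded, with no conditions or evaluators involved.
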StateTree /AICore/Humanoids/StateTrees/Decisions/AlwaysAllowed
states [Root {
  id=#1
  type=STATE
  tasks [{
    id=#2
    description=Delay Task
    properties {
     iD=#3
     instance {_structType=&StateTreeDelayTaskInstanceData, duration=0}
     node {name=Delay Task, _structType=&StateTreeDelayTask}
    }
   }]
  transitions [{id=#4, trigger=ON_STATE_COMPLETED, state {id=#5, name=Root, linkType=SUCCEEDED}}]
 }]
