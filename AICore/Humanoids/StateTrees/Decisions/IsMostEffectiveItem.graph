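# Decision StateTree for humanoid AI bots that checks whether a weapon item is the most effective one against a target.
# A single state gets the target actor, then tasks compute the item's effectiveness and the bot's best weapon effectiveness score.
# On tick the tree succeeds if the item's score is at least the best score, and otherwise fails, as it does when the state fails.
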StateTree /AICore/Humanoids/StateTrees/Decisions/IsMostEffectiveItem
states [
 Root {id=#1, type=STATE, children [#2], tasks [], transitions []}
 Is Most Effective Item {
  id=#2
  type=STATE
  tasks [
   {
    id=#3
    description=Get Target Actor
    properties {
     iD=#4
     instance {_structType=&FortStateTreeGetTargetActorTaskInstanceData}
     node {name=None, _structType=&FortStateTreeGetTargetActorTask}
    }
   }
   {
    id=#5
    description=Bot Get Weapon Item Effectiveness
    properties {
     iD=#6
     instance {
      _structType=/Script/FortniteAIServer.FortStateTreeBotGetWeaponItemEffectivenessTaskInstanceData
      botController=null
      effectiveness=0
      item=null
      target=null
     }
     node {
      name=None
      _structType=/Script/FortniteAIServer.FortStateTreeBotGetWeaponItemEffectivenessTask
      bCanEditConsideredForCompletion=true
      bConsideredForCompletion=true
      bHasTaskCompletionDelegateDispatcher=false
      bTaskEnabled=true
      transitionHandlingPriority=Normal
      bindingsBatch {value=65535}
      executionRuntimeTemplateIndex {value=65535}
      instanceDataHandle {index=65535, source=None, stateHandle {index=65535}}
      instanceTemplateIndex {value=65535}
      outputBindingsBatch {value=65535}
     }
    }
   }
   {
    id=#7
    description=Bot Get Most Effective Weapon Score
    properties {
     iD=#8
     instance {
      _structType=/Script/FortniteAIServer.FortStateTreeBotGetMostEffectiveWeaponScoreTaskInstanceData
      botController=null
      effectiveness=0
      target=null
     }
     node {
      name=None
      _structType=/Script/FortniteAIServer.FortStateTreeBotGetMostEffectiveWeaponScoreTask
      bCanEditConsideredForCompletion=true
      bConsideredForCompletion=true
      bHasTaskCompletionDelegateDispatcher=false
      bTaskEnabled=true
      transitionHandlingPriority=Normal
      bindingsBatch {value=65535}
      executionRuntimeTemplateIndex {value=65535}
      instanceDataHandle {index=65535, source=None, stateHandle {index=65535}}
      instanceTemplateIndex {value=65535}
      outputBindingsBatch {value=65535}
     }
    }
   }
  ]
  transitions [
   {
    id=#9
    trigger=ON_TICK
    conditions [{
      id=#10
      description="Bot Get Weapon Item Effectiveness.Effectiveness >= Bot Get Most Effective Weapon Score.Effectiveness"
      properties {
       iD=#11
       instance {_structType=/Script/StateTreeModule.StateTreeCompareFloatConditionInstanceData, left=0, right=0}
       node {
        name=None
        _structType=/Script/StateTreeModule.StateTreeCompareFloatCondition
        bInvert=false
        deltaIndent=0
        evaluationMode=Evaluated
        operand=And
        operator=GreaterOrEqual
        bindingsBatch {value=65535}
        executionRuntimeTemplateIndex {value=65535}
        instanceDataHandle {index=65535, source=None, stateHandle {index=65535}}
        instanceTemplateIndex {value=65535}
        outputBindingsBatch {value=65535}
       }
      }
     }]
    state {id=#12, name=Root, linkType=SUCCEEDED}
   }
   {id=#13, trigger=ON_TICK, state {id=#12, name=Root, linkType=FAILED}}
   {id=#14, trigger=ON_STATE_FAILED, state {id=#12, name=Root, linkType=FAILED}}
  ]
 }
]
