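# Humanoid AI decision StateTree that checks whether the agent is at its reserved token position.
# Global tasks fetch target, reserved position, actor location and move status.
# While executing, it fails if a distance threshold is exceeded; otherwise it succeeds only when close and path following is idle.
# An optional parameter-gated check runs the IsMostEffectiveItem subtree, before and during execution, failing the decision if that fails.
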StateTree /AICore/Humanoids/StateTrees/Decisions/IsAtReservedPosition
global_tasks [
 {
  id=#1
  description=Get Target Actor
  properties {
   iD=#2
   instance {_structType=&FortStateTreeGetTargetActorTaskInstanceData}
   node {name=Get Target Actor, _structType=&FortStateTreeGetTargetActorTask}
  }
 }
 {
  id=#3
  description=Get Reserved Token Position
  properties {
   iD=#4
   instance {_structType=/Script/FortniteAIServer.FortStateTreeGetReservedTokenPositionTaskInstanceData, pawn=null, target=null, reservedLocation=<0 0 0>}
   node {
    name=Get Reserved Token Position
    _structType=/Script/FortniteAIServer.FortStateTreeGetReservedTokenPositionTask
    bCanEditConsideredForCompletion=true
    bConsideredForCompletion=true
    bHasTaskCompletionDelegateDispatcher=false
    bTaskEnabled=true
    transitionHandlingPriority=Normal
    bindingsBatch {value=65535}
    executionRuntimeTemplateIndex {value=65535}
    instanceDataHandle {index=65535, source=None, stateHandle {index=65535}}
    instanceTemplateIndex {value=65535}
    outputBindingsBatch {value=65535}
   }
  }
 }
 {
  id=#5
  description=Get Actor Location
  properties {
   iD=#6
   instance {
    _structType=/Script/FortniteAIServer.FortGetActorLocationStateTreeTaskInstanceData
    actor=null
    outRefVector {refAccessIndex {value=65535}}
    outVector=<0 0 0>
   }
   node {
    name=Get Actor Location
    _structType=/Script/FortniteAIServer.FortGetActorLocationStateTreeTask
    bCanEditConsideredForCompletion=true
    bConsideredForCompletion=true
    bHasTaskCompletionDelegateDispatcher=false
    bRunOnce=false
    bTaskEnabled=true
    transitionHandlingPriority=Normal
    bindingsBatch {value=65535}
    executionRuntimeTemplateIndex {value=65535}
    instanceDataHandle {index=65535, source=None, stateHandle {index=65535}}
    instanceTemplateIndex {value=65535}
    outputBindingsBatch {value=65535}
   }
  }
 }
 {
  id=#7
  description=Get Move Status
  properties {
   iD=#8
   instance {_structType=/Script/FortniteAIServer.FortStateTreeGetMoveStatusTaskInstanceData, aIController=null, moveStatus=Idle}
   node {
    name=Get Move Status
    _structType=/Script/FortniteAIServer.FortStateTreeGetMoveStatusTask
    bCanEditConsideredForCompletion=true
    bConsideredForCompletion=true
    bHasTaskCompletionDelegateDispatcher=false
    bTaskEnabled=true
    transitionHandlingPriority=Normal
    bindingsBatch {value=65535}
    executionRuntimeTemplateIndex {value=65535}
    instanceDataHandle {index=65535, source=None, stateHandle {index=65535}}
    instanceTemplateIndex {value=65535}
    outputBindingsBatch {value=65535}
   }
  }
 }
]
states [
 Root {
  id=#9
  type=STATE
  children [#10, #11, #12]
  tasks []
  transitions []
 }
 IsMostEffectiveWeapon {
  id=#10
  type=STATE
  enter_conditions [{
    id=#13
    description=Parameters.Check if Most Effective Item is True
    properties {
     iD=#14
     instance {_structType=/Script/StateTreeModule.StateTreeCompareBoolConditionInstanceData, bLeft=false, bRight=true}
     node {
      name=None
      _structType=/Script/StateTreeModule.StateTreeCompareBoolCondition
      bInvert=false
      deltaIndent=0
      evaluationMode=Evaluated
      operand=And
      bindingsBatch {value=65535}
      executionRuntimeTemplateIndex {value=65535}
      instanceDataHandle {index=65535, source=None, stateHandle {index=65535}}
      instanceTemplateIndex {value=65535}
      outputBindingsBatch {value=65535}
     }
    }
   }]
  tasks [{
    id=#15
    description=Run Parallel IsMostEffectiveItem
    properties {
     iD=#16
     instance {
      _structType=/Script/StateTreeModule.StateTreeRunParallelStateTreeTaskInstanceData
      stateTree {stateTree=&/IsMostEffectiveItem, parameters {}, propertyOverrides []}
     }
     node {
      name=None
      _structType=/Script/StateTreeModule.StateTreeRunParallelStateTreeTask
      bCanEditConsideredForCompletion=true
      bConsideredForCompletion=true
      bHasTaskCompletionDelegateDispatcher=false
      bShouldCopyParametersOnExitState=false
      bShouldCopyParametersOnTick=false
      bTaskEnabled=true
      eventHandlingPriority=Normal
      transitionHandlingPriority=Normal
      bindingsBatch {value=65535}
      executionRuntimeTemplateIndex {value=65535}
      instanceDataHandle {index=65535, source=None, stateHandle {index=65535}}
      instanceTemplateIndex {value=65535}
      outputBindingsBatch {value=65535}
      stateTreeOverrideTag {tagName=None}
     }
    }
   }]
  transitions [
   {
    id=#17
    trigger=ON_STATE_SUCCEEDED
    state {id=#18, name=Root, linkType=NEXT_SELECTABLE_STATE}
   }
   {id=#19, trigger=ON_STATE_FAILED, state {id=#18, name=Root, linkType=FAILED}}
  ]
 }
 IsExecuting {
  id=#11
  type=STATE
  children [#20, #21]
  enter_conditions [{
    id=#22
    description=Enum Compare
    properties {
     iD=#23
     instance {
      _structType=/Script/StateTreeModule.StateTreeCompareEnumConditionInstanceData
      left {enum=/Script/FortniteAI.EExecutionStatus, value=0}
      right {enum=&EExecutionStatus, value=3}
     }
     node {
      name=Enum Compare
      _structType=/Script/StateTreeModule.StateTreeCompareEnumCondition
      bInvert=false
      deltaIndent=0
      evaluationMode=Evaluated
      operand=And
      bindingsBatch {value=65535}
      executionRuntimeTemplateIndex {value=65535}
      instanceDataHandle {index=65535, source=None, stateHandle {index=65535}}
      instanceTemplateIndex {value=65535}
      outputBindingsBatch {value=65535}
     }
    }
   }]
  tasks []
  transitions [{
    id=#24
    trigger=ON_TICK
    conditions [{
      id=#25
      description=Distance Compare
      properties {
       iD=#26
       instance {_structType=/Script/StateTreeModule.StateTreeCompareDistanceConditionInstanceData, distance=0, source=<0 0 0>, target=<0 0 0>}
       node {
        name=Distance Compare
        _structType=/Script/StateTreeModule.StateTreeCompareDistanceCondition
        bInvert=false
        deltaIndent=0
        evaluationMode=Evaluated
        operand=And
        operator=Greater
        bindingsBatch {value=65535}
        executionRuntimeTemplateIndex {value=65535}
        instanceDataHandle {index=65535, source=None, stateHandle {index=65535}}
        instanceTemplateIndex {value=65535}
        outputBindingsBatch {value=65535}
       }
      }
     }]
    state {id=#18, name=Root, linkType=FAILED}
   }]
 }
 IsNotExecuting {
  id=#12
  type=STATE
  children [#27, #28]
  tasks []
  transitions []
 }
 IsMostEffectiveWeapon {
  id=#20
  type=STATE
  enter_conditions [{
    id=#29
    description=Parameters.Check if Most Effective Item is True
    properties {
     iD=#30
     instance {_structType=&StateTreeCompareBoolConditionInstanceData, bLeft=false, bRight=true}
     node {
      name=None
      _structType=&StateTreeCompareBoolCondition
      bInvert=false
      deltaIndent=0
      evaluationMode=Evaluated
      operand=And
      bindingsBatch {value=65535}
      executionRuntimeTemplateIndex {value=65535}
      instanceDataHandle {index=65535, source=None, stateHandle {index=65535}}
      instanceTemplateIndex {value=65535}
      outputBindingsBatch {value=65535}
     }
    }
   }]
  tasks [{
    id=#31
    description=Run Parallel IsMostEffectiveItem
    properties {
     iD=#32
     instance {
      _structType=&StateTreeRunParallelStateTreeTaskInstanceData
      stateTree {stateTree=&IsMostEffectiveItem, parameters {}, propertyOverrides []}
     }
     node {
      name=None
      _structType=&StateTreeRunParallelStateTreeTask
      bCanEditConsideredForCompletion=true
      bConsideredForCompletion=true
      bHasTaskCompletionDelegateDispatcher=false
      bShouldCopyParametersOnExitState=false
      bShouldCopyParametersOnTick=false
      bTaskEnabled=true
      eventHandlingPriority=Normal
      transitionHandlingPriority=Normal
      bindingsBatch {value=65535}
      executionRuntimeTemplateIndex {value=65535}
      instanceDataHandle {index=65535, source=None, stateHandle {index=65535}}
      instanceTemplateIndex {value=65535}
      outputBindingsBatch {value=65535}
      stateTreeOverrideTag {tagName=None}
     }
    }
   }]
  transitions [
   {id=#33, trigger=ON_STATE_SUCCEEDED, state {id=#20, name=IsMostEffectiveWeapon}}
   {id=#34, trigger=ON_STATE_FAILED, state {id=#35, name=Root, linkType=FAILED}}
  ]
 }
 DoNothing {id=#21, type=STATE, tasks [], transitions []}
 CloseAndNotMoving {
  id=#27
  type=STATE
  enter_conditions [
   {
    id=#36
    description=Distance Compare
    properties {
     iD=#37
     instance {_structType=&StateTreeCompareDistanceConditionInstanceData, distance=0, source=<0 0 0>, target=<0 0 0>}
     node {
      name=Distance Compare
      _structType=&StateTreeCompareDistanceCondition
      bInvert=false
      deltaIndent=0
      evaluationMode=Evaluated
      operand=And
      operator=LessOrEqual
      bindingsBatch {value=65535}
      executionRuntimeTemplateIndex {value=65535}
      instanceDataHandle {index=65535, source=None, stateHandle {index=65535}}
      instanceTemplateIndex {value=65535}
      outputBindingsBatch {value=65535}
     }
    }
   }
   {
    id=#38
    description=Enum Compare
    properties {
     iD=#39
     instance {
      _structType=&StateTreeCompareEnumConditionInstanceData
      left {enum=/Script/AIModule.EPathFollowingStatus, value=0}
      right {enum=&EPathFollowingStatus, value=0}
     }
     node {
      name=Enum Compare
      _structType=&StateTreeCompareEnumCondition
      bInvert=false
      deltaIndent=0
      evaluationMode=Evaluated
      operand=And
      bindingsBatch {value=65535}
      executionRuntimeTemplateIndex {value=65535}
      instanceDataHandle {index=65535, source=None, stateHandle {index=65535}}
      instanceTemplateIndex {value=65535}
      outputBindingsBatch {value=65535}
     }
    }
   }
  ]
  tasks []
  transitions [{id=#40, trigger=ON_TICK, state {id=#18, name=Root, linkType=SUCCEEDED}}]
 }
 Failure {
  id=#28
  type=STATE
  tasks []
  transitions [{id=#41, trigger=ON_TICK, state {id=#18, name=Root, linkType=FAILED}}]
 }
]
