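# StateTree decision for humanoid AI that evaluates whether the character is outside its leash, with a global task reading the actor location.
# An execution-status check selects IsRunning, which fails the tree on tick once the location is inside the leash.
# Otherwise IsNotRunning gets the move destination and succeeds only if both pawn and destination are outside the leash, else fails.
# Starting therefore has stricter conditions than continuing.
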StateTree /AICore/Humanoids/StateTrees/Decisions/IsOutsideLeash
global_tasks [{
  id=#1
  description=Get Actor Location
  properties {
   iD=#2
   instance {
    _structType=/Script/FortniteAIServer.FortGetActorLocationStateTreeTaskInstanceData
    actor=null
    outRefVector {refAccessIndex {value=65535}}
    outVector=<0 0 0>
   }
   node {
    name=Get Actor Location
    _structType=/Script/FortniteAIServer.FortGetActorLocationStateTreeTask
    bCanEditConsideredForCompletion=true
    bConsideredForCompletion=true
    bHasTaskCompletionDelegateDispatcher=false
    bRunOnce=false
    bTaskEnabled=true
    transitionHandlingPriority=Normal
    bindingsBatch {value=65535}
    executionRuntimeTemplateIndex {value=65535}
    instanceDataHandle {index=65535, source=None, stateHandle {index=65535}}
    instanceTemplateIndex {value=65535}
    outputBindingsBatch {value=65535}
   }
  }
 }]
states [
 Root {
  id=#3
  type=STATE
  children [#4, #5]
  tasks []
  transitions []
 }
 IsRunning {
  id=#4
  type=STATE
  enter_conditions [{
    id=#6
    description=Enum Compare
    properties {
     iD=#7
     instance {
      _structType=/Script/StateTreeModule.StateTreeCompareEnumConditionInstanceData
      left {enum=/Script/FortniteAI.EExecutionStatus, value=0}
      right {enum=&EExecutionStatus, value=3}
     }
     node {
      name=Enum Compare
      _structType=/Script/StateTreeModule.StateTreeCompareEnumCondition
      bInvert=false
      deltaIndent=0
      evaluationMode=Evaluated
      operand=And
      bindingsBatch {value=65535}
      executionRuntimeTemplateIndex {value=65535}
      instanceDataHandle {index=65535, source=None, stateHandle {index=65535}}
      instanceTemplateIndex {value=65535}
      outputBindingsBatch {value=65535}
     }
    }
   }]
  tasks []
  transitions [{
    id=#8
    trigger=ON_TICK
    conditions [{
      id=#9
      description=Is Location Inside Leash
      properties {
       iD=#10
       instance {
        _structType=/Script/FortniteAIServer.FortStateTreeIsInsideLeashConditionInstanceData
        aIController=null
        bUseInnerRadius=false
        location=<0 0 0>
       }
       node {
        name=Is Location Inside Leash
        _structType=/Script/FortniteAIServer.FortStateTreeIsInsideLeashCondition
        bInvert=false
        deltaIndent=0
        evaluationMode=Evaluated
        operand=And
        bindingsBatch {value=65535}
        executionRuntimeTemplateIndex {value=65535}
        instanceDataHandle {index=65535, source=None, stateHandle {index=65535}}
        instanceTemplateIndex {value=65535}
        outputBindingsBatch {value=65535}
       }
      }
     }]
    state {id=#11, name=Root, linkType=FAILED}
   }]
 }
 IsNotRunning {
  id=#5
  type=STATE
  tasks [{
    id=#12
    description=Get Move Destination
    properties {
     iD=#13
     instance {_structType=/Script/FortniteAIServer.FortStateTreeGetMoveDestinationTaskInstanceData, aIController=null, moveDestination=<0 0 0>}
     node {
      name=Get Move Destination
      _structType=/Script/FortniteAIServer.FortStateTreeGetMoveDestinationTask
      bCanEditConsideredForCompletion=true
      bConsideredForCompletion=true
      bHasTaskCompletionDelegateDispatcher=false
      bTaskEnabled=true
      transitionHandlingPriority=Normal
      bindingsBatch {value=65535}
      executionRuntimeTemplateIndex {value=65535}
      instanceDataHandle {index=65535, source=None, stateHandle {index=65535}}
      instanceTemplateIndex {value=65535}
      outputBindingsBatch {value=65535}
     }
    }
   }]
  transitions [
   {
    id=#14
    trigger=ON_TICK
    conditions [
     {
      id=#15
      description=Is Pawn Outside Leash
      properties {
       iD=#16
       instance {
        _structType=&FortStateTreeIsInsideLeashConditionInstanceData
        aIController=null
        bUseInnerRadius=false
        location=<0 0 0>
       }
       node {
        name=Is Pawn Outside Leash
        _structType=&FortStateTreeIsInsideLeashCondition
        bInvert=true
        deltaIndent=0
        evaluationMode=Evaluated
        operand=And
        bindingsBatch {value=65535}
        executionRuntimeTemplateIndex {value=65535}
        instanceDataHandle {index=65535, source=None, stateHandle {index=65535}}
        instanceTemplateIndex {value=65535}
        outputBindingsBatch {value=65535}
       }
      }
     }
     {
      id=#17
      description=Is Destination Outside Leash
      properties {
       iD=#18
       instance {
        _structType=&FortStateTreeIsInsideLeashConditionInstanceData
        aIController=null
        bUseInnerRadius=false
        location=<0 0 0>
       }
       node {
        name=Is Destination Outside Leash
        _structType=&FortStateTreeIsInsideLeashCondition
        bInvert=true
        deltaIndent=0
        evaluationMode=Evaluated
        operand=And
        bindingsBatch {value=65535}
        executionRuntimeTemplateIndex {value=65535}
        instanceDataHandle {index=65535, source=None, stateHandle {index=65535}}
        instanceTemplateIndex {value=65535}
        outputBindingsBatch {value=65535}
       }
      }
     }
    ]
    state {id=#11, name=Root, linkType=SUCCEEDED}
   }
   {id=#19, trigger=ON_TICK, state {id=#11, name=Root, linkType=FAILED}}
  ]
 }
]
