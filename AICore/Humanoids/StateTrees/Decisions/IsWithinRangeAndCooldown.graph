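# Decision StateTree for humanoid AI that gates an action on cooldown and target distance.
# Root selects by ContextData execution status: pending succeeds on tick, while success or error waits out its own cooldown parameter before the range check.
# Other statuses fall through to IsWithinRange, which reads target and pawn locations.
# It succeeds if their distance lies between the Min and Max Range parameters, otherwise it fails.
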StateTree /AICore/Humanoids/StateTrees/Decisions/IsWithinRangeAndCooldown
states [
 Root {
  id=#1
  type=STATE
  children [#2, #3, #4, #5]
  tasks []
  transitions []
 }
 IsExecuting {
  id=#2
  type=STATE
  enter_conditions [{
    id=#6
    description=ContextData.Execution Status is Execution Pending
    properties {
     iD=#7
     instance {
      _structType=/Script/StateTreeModule.StateTreeCompareEnumConditionInstanceData
      left {enum=/Script/FortniteAI.EExecutionStatus, value=0}
      right {enum=&EExecutionStatus, value=3}
     }
     node {
      name=None
      _structType=/Script/StateTreeModule.StateTreeCompareEnumCondition
      bInvert=false
      deltaIndent=0
      evaluationMode=Evaluated
      operand=And
      bindingsBatch {value=65535}
      executionRuntimeTemplateIndex {value=65535}
      instanceDataHandle {index=65535, source=None, stateHandle {index=65535}}
      instanceTemplateIndex {value=65535}
      outputBindingsBatch {value=65535}
     }
    }
   }]
  tasks []
  transitions [{id=#8, trigger=ON_TICK, state {id=#9, name=None, linkType=SUCCEEDED}}]
 }
 HasSucceeded {
  id=#3
  type=STATE
  enter_conditions [{
    id=#10
    description=ContextData.Execution Status is Execution Success
    properties {
     iD=#11
     instance {
      _structType=&StateTreeCompareEnumConditionInstanceData
      left {enum=&EExecutionStatus, value=0}
      right {enum=&EExecutionStatus, value=2}
     }
     node {
      name=None
      _structType=&StateTreeCompareEnumCondition
      bInvert=false
      deltaIndent=0
      evaluationMode=Evaluated
      operand=And
      bindingsBatch {value=65535}
      executionRuntimeTemplateIndex {value=65535}
      instanceDataHandle {index=65535, source=None, stateHandle {index=65535}}
      instanceTemplateIndex {value=65535}
      outputBindingsBatch {value=65535}
     }
    }
   }]
  tasks [{
    id=#12
    description=Delay Parameters.Success Cooldown
    properties {
     iD=#13
     instance {_structType=&StateTreeDelayTaskInstanceData, duration=1}
     node {name=None, _structType=&StateTreeDelayTask}
    }
   }]
  transitions [{id=#14, trigger=ON_STATE_COMPLETED, state {id=#5, name=IsWithinRange}}]
 }
 HasFailed {
  id=#4
  type=STATE
  enter_conditions [{
    id=#15
    description=ContextData.Execution Status is Execution Error
    properties {
     iD=#16
     instance {
      _structType=&StateTreeCompareEnumConditionInstanceData
      left {enum=&EExecutionStatus, value=0}
      right {enum=&EExecutionStatus, value=0}
     }
     node {
      name=None
      _structType=&StateTreeCompareEnumCondition
      bInvert=false
      deltaIndent=0
      evaluationMode=Evaluated
      operand=And
      bindingsBatch {value=65535}
      executionRuntimeTemplateIndex {value=65535}
      instanceDataHandle {index=65535, source=None, stateHandle {index=65535}}
      instanceTemplateIndex {value=65535}
      outputBindingsBatch {value=65535}
     }
    }
   }]
  tasks [{
    id=#17
    description=Delay Parameters.Failure Cooldown
    properties {
     iD=#18
     instance {_structType=&StateTreeDelayTaskInstanceData, duration=1}
     node {name=None, _structType=&StateTreeDelayTask}
    }
   }]
  transitions [{id=#19, trigger=ON_STATE_COMPLETED, state {id=#5, name=IsWithinRange}}]
 }
 IsWithinRange {
  id=#5
  type=STATE
  tasks [
   {
    id=#20
    description=Get Target Actor
    properties {
     iD=#21
     instance {_structType=&FortStateTreeGetTargetActorTaskInstanceData}
     node {name=None, _structType=&FortStateTreeGetTargetActorTask}
    }
   }
   {
    id=#22
    description=Get Target Location
    properties {
     iD=#23
     instance {
      _structType=/Script/FortniteAIServer.FortGetActorLocationStateTreeTaskInstanceData
      actor=null
      outRefVector {refAccessIndex {value=65535}}
      outVector=<0 0 0>
     }
     node {
      name=Get Target Location
      _structType=/Script/FortniteAIServer.FortGetActorLocationStateTreeTask
      bCanEditConsideredForCompletion=true
      bConsideredForCompletion=true
      bHasTaskCompletionDelegateDispatcher=false
      bRunOnce=false
      bTaskEnabled=true
      transitionHandlingPriority=Normal
      bindingsBatch {value=65535}
      executionRuntimeTemplateIndex {value=65535}
      instanceDataHandle {index=65535, source=None, stateHandle {index=65535}}
      instanceTemplateIndex {value=65535}
      outputBindingsBatch {value=65535}
     }
    }
   }
   {
    id=#24
    description=Get Pawn Location
    properties {
     iD=#25
     instance {
      _structType=&FortGetActorLocationStateTreeTaskInstanceData
      actor=null
      outRefVector {refAccessIndex {value=65535}}
      outVector=<0 0 0>
     }
     node {
      name=Get Pawn Location
      _structType=&FortGetActorLocationStateTreeTask
      bCanEditConsideredForCompletion=true
      bConsideredForCompletion=true
      bHasTaskCompletionDelegateDispatcher=false
      bRunOnce=false
      bTaskEnabled=true
      transitionHandlingPriority=Normal
      bindingsBatch {value=65535}
      executionRuntimeTemplateIndex {value=65535}
      instanceDataHandle {index=65535, source=None, stateHandle {index=65535}}
      instanceTemplateIndex {value=65535}
      outputBindingsBatch {value=65535}
     }
    }
   }
  ]
  transitions [
   {
    id=#26
    trigger=ON_TICK
    conditions [
     {
      id=#27
      description="Distance from Get Target Location.Out Vector to Get Pawn Location.Out Vector <= Parameters.Max Range"
      properties {
       iD=#28
       instance {_structType=/Script/StateTreeModule.StateTreeCompareDistanceConditionInstanceData, distance=0, source=<0 0 0>, target=<0 0 0>}
       node {
        name=None
        _structType=/Script/StateTreeModule.StateTreeCompareDistanceCondition
        bInvert=false
        deltaIndent=0
        evaluationMode=Evaluated
        operand=And
        operator=LessOrEqual
        bindingsBatch {value=65535}
        executionRuntimeTemplateIndex {value=65535}
        instanceDataHandle {index=65535, source=None, stateHandle {index=65535}}
        instanceTemplateIndex {value=65535}
        outputBindingsBatch {value=65535}
       }
      }
     }
     {
      id=#29
      description="Distance from Get Target Location.Out Vector to Get Pawn Location.Out Vector >= Parameters.Min Range"
      properties {
       iD=#30
       instance {_structType=&StateTreeCompareDistanceConditionInstanceData, distance=0, source=<0 0 0>, target=<0 0 0>}
       node {
        name=None
        _structType=&StateTreeCompareDistanceCondition
        bInvert=false
        deltaIndent=0
        evaluationMode=Evaluated
        operand=And
        operator=GreaterOrEqual
        bindingsBatch {value=65535}
        executionRuntimeTemplateIndex {value=65535}
        instanceDataHandle {index=65535, source=None, stateHandle {index=65535}}
        instanceTemplateIndex {value=65535}
        outputBindingsBatch {value=65535}
       }
      }
     }
    ]
    state {id=#9, name=Root, linkType=SUCCEEDED}
   }
   {id=#31, trigger=ON_TICK, state {id=#9, name=Root, linkType=FAILED}}
   {id=#32, trigger=ON_STATE_COMPLETED, state {id=#9, name=None, linkType=FAILED}}
  ]
 }
]
