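# Decision StateTree for humanoid AI that reports whether the pawn needs healing.
# Its single Evaluate state runs Get Health and Get Shield tasks to read health and shield percentages.
# On tick, it succeeds if either percentage is at or below its minimum threshold parameter, and otherwise an unconditional fallback transition fails the tree, making it a yes/no check.
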StateTree /AICore/Humanoids/StateTrees/Decisions/NeedsHealing
states [
 Root {id=#1, type=STATE, children [#2], tasks [], transitions []}
 Evaluate {
  id=#2
  type=STATE
  tasks [
   {
    id=#3
    description=Get Health
    properties {
     iD=#4
     instance {_structType=/Script/FortniteAIServer.FortStateTreeGetHealthTaskInstanceData, currentHealth=0, healthPercentage=0, maxHealth=0, pawn=null}
     node {
      name=None
      _structType=/Script/FortniteAIServer.FortStateTreeGetHealthTask
      bCanEditConsideredForCompletion=true
      bConsideredForCompletion=true
      bHasTaskCompletionDelegateDispatcher=false
      bTaskEnabled=true
      transitionHandlingPriority=Normal
      bindingsBatch {value=65535}
      executionRuntimeTemplateIndex {value=65535}
      instanceDataHandle {index=65535, source=None, stateHandle {index=65535}}
      instanceTemplateIndex {value=65535}
      outputBindingsBatch {value=65535}
     }
    }
   }
   {
    id=#5
    description=Get Shield
    properties {
     iD=#6
     instance {_structType=/Script/FortniteAIServer.FortStateTreeGetShieldTaskInstanceData, currentShield=0, maxShield=0, pawn=null, shieldPercentage=0}
     node {
      name=None
      _structType=/Script/FortniteAIServer.FortStateTreeGetShieldTask
      bCanEditConsideredForCompletion=true
      bConsideredForCompletion=true
      bHasTaskCompletionDelegateDispatcher=false
      bTaskEnabled=true
      transitionHandlingPriority=Normal
      bindingsBatch {value=65535}
      executionRuntimeTemplateIndex {value=65535}
      instanceDataHandle {index=65535, source=None, stateHandle {index=65535}}
      instanceTemplateIndex {value=65535}
      outputBindingsBatch {value=65535}
     }
    }
   }
  ]
  transitions [
   {
    id=#7
    trigger=ON_TICK
    conditions [
     {
      id=#8
      description="Get Health.Health Percentage <= Parameters.Min Health Percentage"
      properties {
       iD=#9
       instance {_structType=/Script/StateTreeModule.StateTreeCompareFloatConditionInstanceData, left=0, right=0}
       node {
        name=None
        _structType=/Script/StateTreeModule.StateTreeCompareFloatCondition
        bInvert=false
        deltaIndent=0
        evaluationMode=Evaluated
        operand=And
        operator=LessOrEqual
        bindingsBatch {value=65535}
        executionRuntimeTemplateIndex {value=65535}
        instanceDataHandle {index=65535, source=None, stateHandle {index=65535}}
        instanceTemplateIndex {value=65535}
        outputBindingsBatch {value=65535}
       }
      }
     }
     {
      id=#10
      description="Get Shield.Shield Percentage <= Parameters.Min Shield Percentage"
      properties {
       expressionOperand=Or
       iD=#11
       instance {_structType=&StateTreeCompareFloatConditionInstanceData, left=0, right=0}
       node {
        name=None
        _structType=&StateTreeCompareFloatCondition
        bInvert=false
        deltaIndent=0
        evaluationMode=Evaluated
        operand=And
        operator=LessOrEqual
        bindingsBatch {value=65535}
        executionRuntimeTemplateIndex {value=65535}
        instanceDataHandle {index=65535, source=None, stateHandle {index=65535}}
        instanceTemplateIndex {value=65535}
        outputBindingsBatch {value=65535}
       }
      }
     }
    ]
    state {id=#12, name=Root, linkType=SUCCEEDED}
   }
   {id=#13, trigger=ON_TICK, state {id=#12, name=Root, linkType=FAILED}}
  ]
 }
]
