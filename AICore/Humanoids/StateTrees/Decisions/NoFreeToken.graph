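# StateTree decision for humanoid AI that checks whether the target has no free Athena.AI.Token.Damage token.
# A global Get Target Actor task runs alongside two ordered states under Root.
# HasToken is entered when the free-token condition passes and fails the tree on tick.
# Otherwise selection falls through to the unconditional HasNoToken, which succeeds the tree on tick, so the result inverts the condition.
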StateTree /AICore/Humanoids/StateTrees/Decisions/NoFreeToken
global_tasks [{
  id=#1
  description=Get Target Actor
  properties {
   iD=#2
   instance {_structType=&FortStateTreeGetTargetActorTaskInstanceData}
   node {name=Get Target Actor, _structType=&FortStateTreeGetTargetActorTask}
  }
 }]
states [
 Root {
  id=#3
  type=STATE
  children [#4, #5]
  tasks []
  transitions []
 }
 HasToken {
  id=#4
  type=STATE
  enter_conditions [{
    id=#6
    description=Target has free token
    properties {
     iD=#7
     instance {
      _structType=/Script/FortniteAIServer.FortStateTreeHasFreeTokenConditionInstanceData
      reserverPawn=null
      target=null
      tokenTagQuery {
       autoDescription=" ANY( Athena.AI.Token.Damage )"
       tokenStreamVersion=0
       userDescription=""
       queryTokenStream [0,1,1,1,0]
       tagDictionary [{tagName=Athena.AI.Token.Damage}]
      }
     }
     node {
      name=Target has free token
      _structType=/Script/FortniteAIServer.FortStateTreeHasFreeTokenCondition
      deltaIndent=0
      evaluationMode=Evaluated
      operand=And
      bindingsBatch {value=65535}
      executionRuntimeTemplateIndex {value=65535}
      instanceDataHandle {index=65535, source=None, stateHandle {index=65535}}
      instanceTemplateIndex {value=65535}
      outputBindingsBatch {value=65535}
     }
    }
   }]
  tasks []
  transitions [{id=#8, trigger=ON_TICK, state {id=#9, name=None, linkType=FAILED}}]
 }
 HasNoToken {
  id=#5
  type=STATE
  tasks []
  transitions [{id=#10, trigger=ON_TICK, state {id=#9, name=None, linkType=SUCCEEDED}}]
 }
]
